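# Material for the VR Editor teleport direction visual, judging by its name and folder.
# A Color vector parameter, light blue by default, feeds an EyeAdaptationInverse node at full strength to counteract auto exposure.
# The material sets the disable depth test flag and lists Default Lit shading, with no textures or other logic in the graph.
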Material /Engine/VREditor/Teleport/TeleportDirectionMaterial
graph {
 properties {bDisableDepthTest=true, bUsedWithStaticLighting=true, pixelDepthOffsetMode=PDOM_Legacy, usedShadingModels=Default Lit}
 nodes [EyeAdaptationInverse(LightValueInput=VectorParameter() {parameterName=Color, defaultValue=<rgba 0 0.3683 0.82 0>}.RGB, AlphaInput=Constant() {r=1})]
}
